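# Unlit, two-sided masked material for the Vector Spark emote background on skeletal meshes, drawing animated, tinted noise inside a soft spherical mask.
# A panning distortion normal warps swizzled, panning UV1 noise, with the warp skipped on mobile.
# The sphere mask, whose radius switches through a front-end detector parameter collection, shapes that noise.
# The result drives dithered temporal-AA alpha and a colored emissive adjusted by time-of-day curve and FX functions.
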Material /BRCosmetics/Animation/Props/Gadgets_VectorSpark/Materials/M_VectorSpark_Emote_Background
graph {
 properties {
  bUsedWithMeshDeformer=true
  bUsedWithSkeletalMesh=true
  blendMode=BLEND_Masked
  pixelDepthOffsetMode=PDOM_Legacy
  shadingModel=MSM_Unlit
  twoSided=true
  usedShadingModels=Unlit
 }
 nodes [
  n1 = VectorParameter() {group=Trailing Edge Sphere Mask, parameterName=Sphere Mask UV Scale, sortPriority=4, defaultValue=<rgba 1 1 0 0>}
  n2 = VectorParameter() {group=Trailing Edge Sphere Mask, parameterName="Sphere Mask Coord (B)", sortPriority=3, defaultValue=<rgba 0.685 0.335 0 0>}
  n3 = SphereMask(
   A=Multiply(A=TextureCoordinate(), B=ComponentMask(None=@n1.RGB) {g=true, r=true})
   B=ComponentMask(None=@n2.RGB) {g=true, r=true}
   Radius=LinearInterpolate(
    A=ScalarParameter() {defaultValue=0.28, group=Trailing Edge Sphere Mask, parameterName=Sphere Mask Radius, sortPriority=1}
    B=ScalarParameter() {defaultValue=0.28, group=Trailing Edge Sphere Mask, parameterName=Sphere Mask Radius FE, sortPriority=1}
    Alpha=CollectionParameter() {collection=/BRCosmetics/Effects/FrontEndDetecter/MPC_FrontEndDetecter, parameterName=bIsFrontEnd}
   )
   Hardness=ScalarParameter() {defaultValue=0.8, group=Trailing Edge Sphere Mask, parameterName=Sphere Mask Hardness, sortPriority=2}
  )
  n4 = TextureCoordinate() {vTiling=1.5}
  n5 = ComponentMask(None=VectorParameter() {group=Distortion, parameterName=Distortion UV Scale, sortPriority=3, defaultValue=<rgba 2 2 0 0>}.RGB) {g=true, r=true}
  n6 = Time()
  n7 = ScalarParameter() {defaultValue=1, group=Panner Speed, parameterName=Speed}
  n8 = TextureSampleParameter2D(UVs=Panner(Coordinate=Multiply(A=@n4, B=@n5), Time=Multiply(A=@n6, B=@n7)) {speedX=0.025, speedY=0.05}) {
   group=Distortion
   parameterName=Distortion Normal
   samplerType=SAMPLERTYPE_Normal
   sortPriority=2
   texture=&T_Weap_Distort_N_02
  }
  n9 = MaterialFunctionCall(Normal=@n8.RGB, Flatness=ScalarParameter() {defaultValue=-2, group=Distortion, parameterName=UV_Normal_Flatness, sortPriority=4}) {materialFunction=&FlattenNormal}
  n10 = LinearInterpolate(
   B=Divide(
    A=ComponentMask(None=@n9.Result) {g=true, r=true}
    B=ScalarParameter() {defaultValue=4, group=Distortion, parameterName=UV_Distortion_Divsor, sortPriority=1}
   ) {constB=20}
   Alpha=ScalarParameter() {defaultValue=1, group=Distortion, parameterName=UV_Distortion_Lerp}
  )
  n11 = TextureCoordinate() {coordinateIndex=1, uTiling=0.8, vTiling=0.9}
  n12 = ComponentMask(None=VectorParameter() {group=Fire Shape, parameterName=Fire_Micro_UVScale, sortPriority=8, defaultValue=<rgba 1 1 0 0>}.RGB) {g=true, r=true}
  n13 = Panner(
   Coordinate=MaterialFunctionCall(XY=Multiply(A=@n11, B=@n12)) {materialFunction=&Swizzle}.YX
   Time=Multiply(A=Time(), B=ScalarParameter() {defaultValue=1, group=Panner Speed, parameterName=Speed})
  ) {speedX=0.05, speedY=0.025}
  n14 = TextureSampleParameter2D(UVs=ShadingPathSwitch(Default=Add(A=@n10, B=@n13), Mobile=@n13) {desc=Migrated from FeatureLevelSwitch}) {
   group=Fire Shape
   parameterName=Fire Micro Texture
   samplerType=SAMPLERTYPE_LinearColor
   sortPriority=6
   texture=&T_NoisePack_LinearColor
  }
  n15 = Multiply(A=Clamp(None=@n14.R) {minDefault=0.5}, B=@n3)
  MaterialFunctionCall(Alpha Threshold=LinearInterpolate(A=@n15, Alpha=@n3) {constA=1}) {materialFunction=&DitherTemporalAA_Fortnite}
  n16 = VectorParameter() {parameterName=Color, defaultValue=<rgba 1.862 2.3565 2.5 1>}
  MaterialFunctionCall(Emissive=MaterialFunctionCall!(Color=Multiply(A=@n16.RGB, B=@n15)) {materialFunction=/Game/Packages/Fortress_SharedMaterials/Anime_Master_Material/MF_TODColorCurve}.Result) {materialFunction=&MF_FX_EmissiveAdjustments}
 ]
}
