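# Unlit masked material for the VectorSpark gadget emote prop's shadow variant, usable on skeletal meshes and mesh deformers.
# A sphere mask on transformed world position feeds temporal-AA dithering, producing a soft dithered cutout.
# The front-end detector parameter collection blends the sphere's offset and radius between default and front-end values.
# A separate branch samples the VectorSpark E texture and adjusts it with a cheap contrast function.
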Material /BRCosmetics/Animation/Props/Gadgets_VectorSpark/Materials/M_VectorSpark_Emote_Shadow
graph {
 properties {
  bUsedWithMeshDeformer=true
  bUsedWithSkeletalMesh=true
  blendMode=BLEND_Masked
  pixelDepthOffsetMode=PDOM_Legacy
  shadingModel=MSM_Unlit
  usedShadingModels=Unlit
 }
 nodes [
  n1 = TransformPosition(None=WorldPosition().XYZ) {transformSourceType=TRANSFORMPOSSOURCE_World}
  n2 = CollectionParameter() {collection=/BRCosmetics/Effects/FrontEndDetecter/MPC_FrontEndDetecter, parameterName=bIsFrontEnd}
  n3 = ComponentMask(None=LinearInterpolate(
    A=VectorParameter() {group=Sphere Mask, parameterName=SphereMask Offset, defaultValue=<rgba -54 87 138 0>}.RGB
    B=VectorParameter() {group=Sphere Mask, parameterName=SphereMask FE Offset, defaultValue=<rgba -54 87 138 0>}.RGB
    Alpha=@n2
   )) {b=true, g=true, r=true}
  n4 = LinearInterpolate(
   A=ScalarParameter() {defaultValue=33.5, group=Sphere Mask, parameterName=Sphere Mask Radius}
   B=ScalarParameter() {defaultValue=33.5, group=Sphere Mask, parameterName=Sphere Mask Radius FE}
   Alpha=@n2
  )
  n5 = ScalarParameter() {defaultValue=0.9, group=Sphere Mask, parameterName=Hardness}
  MaterialFunctionCall(Alpha Threshold=LinearInterpolate(Alpha=SphereMask(A=@n1, B=@n3, Radius=@n4, Hardness=@n5) {attenuationRadius=100, hardnessPercent=1})) {materialFunction=&DitherTemporalAA_Fortnite}
  MaterialFunctionCall!(
   In=TextureSample() {texture=/BRCosmetics/Animation/Props/Gadgets_VectorSpark/Textures/T_VectorSpark_E}.RGB
   Contrast=ScalarParameter() {defaultValue=4, parameterName="Contrast (S)"}
  ) {materialFunction=/Engine/Functions/Engine_MaterialFunctions01/ImageAdjustment/CheapContrast_RGB}
 ]
}
